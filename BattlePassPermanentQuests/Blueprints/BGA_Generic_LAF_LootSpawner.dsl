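; Short-lived Battle Pass quest actor that spawns rare Apollo treasure loot on the server and marks its rarest pickup for the owning player.
; On BeginPlay it spawns the loot, plays replicated Niagara and sound, then starts marking after a delay.
; Marking maps pickups by rarity tag, short-circuits on Transcendent, otherwise resolves through an ordered tag array.
; The bDoneMarking RepNotify destroys the actor on authority, with a lifespan also set.
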(event EventBeginPlay
  (bind _returnvalue (Actor|GetOwner))
  ;[0 0 0] Spawn the loot and call for client effects. Destroy actor when marking is done
  (|CallParentFunction)
  (Networking|SwitchHasAuthority
    (:Authority
      (Utilities|IsValid _returnvalue
        (:Is%20Valid
          (bind _itempickupsspawned (Class|AthenaFunctionLibrary|SpawnLootfromTable "Loot_ApolloTreasure_Rare" 1 -1 (Transformation|GetActorLocation) (Transformation|GetActorForwardVector _returnvalue) 0 true true))
          (Variables|Default|SetLootSpawned _itempickupsspawned)
          (CallFunction|PlaySoundandFX)
          (Utilities|FlowControl|Delay 1.0)
          (CallFunction|MarkRarestLootItem)
          (Actor|SetLifeSpan 5.0))
        (:Is%20Not%20Valid
          (Variables|Default|SetbDoneMarking true))))))

(event Custom|PlaySoundAndFX
  ;[0 0 0] Play VFX and sound - replicate to all clients
  (Niagara|SpawnSystemAttached (Variables|Default|GetLootSpawnVFX) (Variables|Transformation|GetRootComponent) "None" "0, 0, 0" "0, 0, 0" "KeepRelativeOffset" true)
  (Audio|SpawnSoundAttached (Variables|Default|GetLootSpawnSound) (Variables|Transformation|GetRootComponent)))

(event Custom|MarkRarestLootItem
  (bind _allorderedraritytags (Variables|Default|GetAllOrderedRarityTags))
  (bind _returnvalue (GameplayTags|MakeGameplayTagContainerfromArray _allorderedraritytags))
  (bind _itemtoraritytagmap (Variables|Default|GetItemToRarityTagMap))
  (bind _currentpickup (Variables|Default|GetCurrentPickup))
  ;[0 0 0] Mark the rarest loot item spawned
  (bind _array_element (Utilities|Array|ForEachLoopwithBreak (Variables|Default|GetLootSpawned))
    (:LoopBody
      (bind _asfort_pickup_athena (Utilities|Casting|CastToFortPickupAthena _array_element)
        (:then
          (bind _output_get (Variables|Default|SetCurrentPickup _asfort_pickup_athena))
          ;; Find the rarity tag on the item if it exists. If we've found an item of that rarity already, don't consider this one
          (if (not (GameplayTags|HasAnyTags (GameplayTags|Filter (GameplayTags|GetOwnedGameplayTags _output_get :!class /Script/GameplayTags.BlueprintGameplayTagLibrary) _returnvalue true) (Variables|Default|GetFoundRarityTags) true))
            (if (== (GameplayTags|GetNumGameplayTagsinContainer (GameplayTags|Filter (GameplayTags|GetOwnedGameplayTags _output_get :!class /Script/GameplayTags.BlueprintGameplayTagLibrary) _returnvalue true)) 1)
              (if (GameplayTags|IsGameplayTagValid (Utilities|Array|Get(acopy) (GameplayTags|BreakGameplayTagContainer (GameplayTags|Filter (GameplayTags|GetOwnedGameplayTags _output_get :!class /Script/GameplayTags.BlueprintGameplayTagLibrary) _returnvalue true))))
                (bind _output_get_2 (Variables|Default|SetTagFound (Utilities|Array|Get(acopy) (GameplayTags|BreakGameplayTagContainer (GameplayTags|Filter (GameplayTags|GetOwnedGameplayTags _output_get :!class /Script/GameplayTags.BlueprintGameplayTagLibrary) _returnvalue true)))))
                ;; Add the item/rarity tag pair to a map for tracking unless it is transcendent, in which case we assume it is the most rare item
                (switch Utilities|FlowControl|Switch|SwitchonGameplayTag _output_get_2
                  (:Default
                    (if (not (GameplayTags|HasTag (Variables|Default|GetFoundRarityTags) (Variables|Default|GetTagFound) true))
                      (GameplayTags|AddGameplayTag (Variables|Default|GetFoundRarityTags) (Variables|Default|GetTagFound) :!class /Script/GameplayTags.BlueprintGameplayTagLibrary)
                      (Utilities|Map|Add (Variables|Default|GetItemToRarityTagMap) (Variables|Default|GetTagFound) _currentpickup)))
                  (:Rarity.Transcendent
                    (Variables|Default|SetRarestItem _currentpickup)
                    (CallFunction|StopSearching)))))))
        (:CastFailed
          (Variables|Default|SetbDoneMarking true))))
    (:Completed
      ;; Using the ordered array of rarity tags, find the item from the map with the highest rarity and mark it unless only one item was found
      (if (> (Utilities|Map|Length (Variables|Default|GetItemToRarityTagMap)) 0)
        (bind _array_element_2 (Utilities|Array|ForEachLoopwithBreak (Variables|Default|GetAllOrderedRarityTags))
          (:LoopBody
            (if (.ReturnValue (Utilities|Map|Find _itemtoraritytagmap _array_element_2))
              (Variables|Default|SetRarestItem (Utilities|Map|Find _itemtoraritytagmap _array_element_2))))
          (:Completed
            _))
        (else
          _))
      (Utilities|IsValid (Variables|Default|GetRarestItem)
        (:Is%20Valid
          (MarkerComponent|MarkActoronServer (Class|FortPlayerController|GetMarkerComponent (Fortnite|GetFortPlayerControllerfromActor (Actor|GetOwner))) (Variables|Default|GetRarestItem) true true)
          (Variables|Default|SetbDoneMarking true))
        (:Is%20Not%20Valid
          (Variables|Default|SetbDoneMarking true))))))

(event Custom|StopSearching
  (bind _allorderedraritytags (Variables|Default|GetAllOrderedRarityTags))
  (bind _returnvalue (GameplayTags|MakeGameplayTagContainerfromArray _allorderedraritytags))
  (bind _itemtoraritytagmap (Variables|Default|GetItemToRarityTagMap))
  (bind _currentpickup (Variables|Default|GetCurrentPickup))
  ;[0 0 0] Mark the rarest loot item spawned
  (bind _array_element (Utilities|Array|ForEachLoopwithBreak (Variables|Default|GetLootSpawned))
    (:LoopBody
      (bind _asfort_pickup_athena (Utilities|Casting|CastToFortPickupAthena _array_element)
        (:then
          (bind _output_get (Variables|Default|SetCurrentPickup _asfort_pickup_athena))
          ;; Find the rarity tag on the item if it exists. If we've found an item of that rarity already, don't consider this one
          (if (not (GameplayTags|HasAnyTags (GameplayTags|Filter (GameplayTags|GetOwnedGameplayTags _output_get :!class /Script/GameplayTags.BlueprintGameplayTagLibrary) _returnvalue true) (Variables|Default|GetFoundRarityTags) true))
            (if (== (GameplayTags|GetNumGameplayTagsinContainer (GameplayTags|Filter (GameplayTags|GetOwnedGameplayTags _output_get :!class /Script/GameplayTags.BlueprintGameplayTagLibrary) _returnvalue true)) 1)
              (if (GameplayTags|IsGameplayTagValid (Utilities|Array|Get(acopy) (GameplayTags|BreakGameplayTagContainer (GameplayTags|Filter (GameplayTags|GetOwnedGameplayTags _output_get :!class /Script/GameplayTags.BlueprintGameplayTagLibrary) _returnvalue true))))
                (bind _output_get_2 (Variables|Default|SetTagFound (Utilities|Array|Get(acopy) (GameplayTags|BreakGameplayTagContainer (GameplayTags|Filter (GameplayTags|GetOwnedGameplayTags _output_get :!class /Script/GameplayTags.BlueprintGameplayTagLibrary) _returnvalue true)))))
                ;; Add the item/rarity tag pair to a map for tracking unless it is transcendent, in which case we assume it is the most rare item
                (switch Utilities|FlowControl|Switch|SwitchonGameplayTag _output_get_2
                  (:Default
                    (if (not (GameplayTags|HasTag (Variables|Default|GetFoundRarityTags) (Variables|Default|GetTagFound) true))
                      (GameplayTags|AddGameplayTag (Variables|Default|GetFoundRarityTags) (Variables|Default|GetTagFound) :!class /Script/GameplayTags.BlueprintGameplayTagLibrary)
                      (Utilities|Map|Add (Variables|Default|GetItemToRarityTagMap) (Variables|Default|GetTagFound) _currentpickup)))
                  (:Rarity.Transcendent
                    (Variables|Default|SetRarestItem _currentpickup)
                    (CallFunction|StopSearching)))))))
        (:CastFailed
          (Variables|Default|SetbDoneMarking true))))
    (:Completed
      ;; Using the ordered array of rarity tags, find the item from the map with the highest rarity and mark it unless only one item was found
      (if (> (Utilities|Map|Length (Variables|Default|GetItemToRarityTagMap)) 0)
        (bind _array_element_2 (Utilities|Array|ForEachLoopwithBreak (Variables|Default|GetAllOrderedRarityTags))
          (:LoopBody
            (if (.ReturnValue (Utilities|Map|Find _itemtoraritytagmap _array_element_2))
              (Variables|Default|SetRarestItem (Utilities|Map|Find _itemtoraritytagmap _array_element_2))))
          (:Completed
            _))
        (else
          _))
      (Utilities|IsValid (Variables|Default|GetRarestItem)
        (:Is%20Valid
          (MarkerComponent|MarkActoronServer (Class|FortPlayerController|GetMarkerComponent (Fortnite|GetFortPlayerControllerfromActor (Actor|GetOwner))) (Variables|Default|GetRarestItem) true true)
          (Variables|Default|SetbDoneMarking true))
        (:Is%20Not%20Valid
          (Variables|Default|SetbDoneMarking true))))))

(event EventEndPlay
  (params (EndPlayReason :type EEndPlayReason%20Enum :default "Destroyed"))
  (|CallParentFunction EndPlayReason))

(fn CheckIfRarestTag
  (params (GameplayTagToCheck :type Boolean)))

(fn OnRep_bDoneMarking
  (Networking|SwitchHasAuthority
    (:Authority
      (Actor|DestroyActor))))

(fn ConstructionScript)
